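# Looping ambient sound cue for a scanning server rack prop, routed to the Athena ambient props sound class.
# A mixer layers the server scanning loop, at double weight, over a console terminal hum loop pitched down through a modulator.
# An enveloper at the root ramps volume up from silence.
# The cue uses server rack hum attenuation at low priority and overrides concurrency to cap simultaneous instances.
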SoundCue /Game/Sounds/Helios/Environment_Actors/ServerRacks/Spot_ServeRack_Scanning_Loop_03_Cue
root=Enveloper_0
nodes [
 {
  id=Enveloper_0
  children [Mixer_0]
  properties {pitchCurve {editorCurveData {keys [{time=0}]}}, volumeCurve {editorCurveData {keys [<at 0 0>,<at 1 1>]}}}
 }
 {id=Mixer_0, children [Modulator_0,Modulator_1], properties {inputVolume [2,1]}}
 {id=Modulator_0, children [WavePlayer_1]}
 {id=Modulator_1, children [WavePlayer_0], properties {pitchMax=0.8, pitchMin=0.8}}
 {
  id=WavePlayer_0
  properties {bLooping=true, soundWaveAssetPtr=/Game/Sounds/Helios/Environment_Actors/ConsoleTerminals/Spot_Console_Terminal_Communication_Hum_Loop_02}
 }
 {
  id=WavePlayer_1
  properties {bLooping=true, soundWaveAssetPtr=&/Spot_Server_Stack_Scanning_Loop_03}
 }
]
properties {
 attenuationSettings=&/Spot_ServerRack_Hum_Att
 bOverrideConcurrency=true
 maxDistance=850
 priority=0.25
 soundClassObject=&Fort_Ambient_Props_Athena
 volumeMultiplier=0.301
 concurrencyOverrides {maxCount=3}
}
